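; Blueprint notification shown in the frontend Compete UI, apparently when a bookmarked tournament goes live.
; When the notification's take-action event fires, it fetches the FortCompeteUIManager game instance subsystem.
; It passes itself to that manager's OnFrontendBookmarkedTournamentNotificationAction, leaving all handling of the action to the manager.
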(event Notification|EventOnTakeAction
  (Class|FortCompeteUIManager|OnFrontendBookmarkedTournamentNotificationAction (GameInstanceSubsystems|GetFortCompeteUIManager) self))
