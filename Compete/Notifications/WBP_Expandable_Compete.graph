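# Expandable lobby widget that shows bookmarked Compete tournaments with a countdown, count badge and live state.
# A one-second timer prunes ended view-model entries, updates the countdown and checks the first entry, firing a notification when it goes live.
# Hover, or mobile setup, cycles multiple bookmarks on a one-shot timer, crossfading two name texts and advancing pips.
# Clicking enters a live tournament's lobby or opens an upcoming one's details.
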WidgetBlueprint /Compete/Notifications/WBP_Expandable_Compete
animations [NextFrameAnim {
  binding_count=3
  length_seconds=0.5989
  object=&self:NextFrameAnim
  sequence {bindings [
    Timer {
     id=#1
     possessed_class=CommonTextBlock
     tracks [{
       type=MovieSceneFloatTrack
       display_name=Render Opacity
       sections [{type=MovieSceneFloatSection, end_frame=11, start_frame=0, channels [None {type=MovieSceneScriptingFloatChannel, keys [[0,0.3],[11.9763,1]]}]}]
      }]
    }
    TournamentName_Bookmarked {
     id=#2
     possessed_class=CommonTextBlock
     tracks [
      {
       type=MovieSceneFloatTrack
       display_name=Render Opacity
       sections [{type=MovieSceneFloatSection, channels [None {type=MovieSceneScriptingFloatChannel, thinned_from=4, keys [[0,1],[5.8,0],[11.9767,0]]}]}]
      }
      {
       type=MovieScene2DTransformTrack
       sections [{
         type=MovieScene2DTransformSection
         end_frame=11
         start_frame=0
         channels [
          Translation.X {type=MovieSceneScriptingFloatChannel, keys [[0,0],[11.686,-15],[11.9767,0]]}
          Translation.Y {type=MovieSceneScriptingFloatChannel, const [0,1]}
          Angle {type=MovieSceneScriptingFloatChannel, const [0,1]}
          Scale.X {type=MovieSceneScriptingFloatChannel, const [1,1]}
          Scale.Y {type=MovieSceneScriptingFloatChannel, const [1,1]}
          Shear.X {type=MovieSceneScriptingFloatChannel, const [0,1]}
          Shear.Y {type=MovieSceneScriptingFloatChannel, const [0,1]}
         ]
        }]
      }
     ]
    }
    TournamentName_Bookmarked_ForAnim {
     id=#3
     possessed_class=CommonTextBlock
     tracks [
      {
       type=MovieSceneFloatTrack
       display_name=Render Opacity
       sections [{type=MovieSceneFloatSection, channels [None {type=MovieSceneScriptingFloatChannel, keys [[0,0],[4.9603,0],[11.6867,1],[11.9767,1]]}]}]
      }
      {
       type=MovieScene2DTransformTrack
       sections [{
         type=MovieScene2DTransformSection
         channels [
          Translation.X {type=MovieSceneScriptingFloatChannel, keys [[0,0],[3.9923,30],[11.9767,0]]}
          Translation.Y {type=MovieSceneScriptingFloatChannel, const [0,1]}
          Angle {type=MovieSceneScriptingFloatChannel, const [0,1]}
          Scale.X {type=MovieSceneScriptingFloatChannel, const [1,1]}
          Scale.Y {type=MovieSceneScriptingFloatChannel, const [1,1]}
          Shear.X {type=MovieSceneScriptingFloatChannel, const [0,1]}
          Shear.Y {type=MovieSceneScriptingFloatChannel, const [0,1]}
         ]
        }]
      }
     ]
    }
   ]}
 }]
graphs [
 ConstructVM {nodes [
   ConstructVM() {is=title K2Node_FunctionEntry}
   n1 = GetBookmarksVM!(self:Fort Compete UIManager Object Reference=Get FortCompeteUIManager() -> (ReturnValue:Fort Compete UIManager Object Reference) {is=title K2Node_GetSubsystem}) -> (ReturnValue:Tournament Bookmarks VM Object Reference)
   > InitializeVM(self:Tournament Bookmarks VM Object Reference=@n1)
   > SetFortTournamentBookmarksVM!(execute, self, Viewmodel:Tournament Bookmarks VM Object Reference=@n1) -> (then)
  ]}
 EventGraph {nodes [
   n1 = StartTimer!() -> (OutputDelegate, then, Time:double=~) {is=title K2Node_CustomEvent}
   > n2 = Set Timer by Event(Delegate=Create Event(), Time=@n1.Time, bLooping=false)
   > Set TimerHandle(TimerHandle:Timer Handle Structure=@n2) -> (Output_Get:Timer Handle Structure)
   ShowNextFrame() {is=title K2Node_CustomEvent}
   > ChangeFrame!(execute, self, bPreviousFrame:bool=false) -> (then)
   CountdownTickEverySecond() {is=title K2Node_CustomEvent}
   n3 = Get FortTournamentBookmarksVM() -> (FortTournamentBookmarksVM:Tournament Bookmarks VM Object Reference)
   > n4 = Is Valid(InputObject=@n3)
   > n5 = RemoveEndedEntries(self:Tournament Bookmarks VM Object Reference=@n3)
   n6 = Branch(
    execute=[@n5.then, @n4.Is Not Valid]
    Condition=Is Not Empty!(TargetArray:Array of Fort Upcoming Tournament Data Structures=Get Bookmarked Events() -> (Bookmarked Events:Array of Fort Upcoming Tournament Data Structures)) -> (ReturnValue:bool)
   )
   > UpdateCountdownText!(execute, self, ShowMinTime:bool=Get bIsExtPartCollapsed() -> (bIsExtPartCollapsed:bool)) -> (then)
   > UpdateLiveState()
   UpdateVisibility(execute=@n6.else)
   n7 = ChangeFrame!() -> (OutputDelegate, then, bPreviousFrame:bool=~) {is=title K2Node_CustomEvent}
   n8 = Get bIsExtPartCollapsed()
   > n9 = Branch(Condition=AND Boolean(A=Get bMultipleBookmarked() -> (bMultipleBookmarked:bool), B=NOT Boolean(A=@n8)))
   > IncrementIndex!(execute, self, PreviousFrame:bool=@n7.bPreviousFrame) -> (then, NewIndex:int=0)
   > UpdateAnimText()
   > UpdateCountdownText(ShowMinTime=false)
   > n10 = "Play Animation with Finished event (legacy)"!(
    execute:Exec
    Widget:User Widget Object Reference=Self-Reference()
    InAnimation:Widget Animation Object Reference=Get NextFrameAnim() -> (NextFrameAnim:Widget Animation Object Reference)
    StartAtTime:float=0
    NumLoopsToPlay:int=1
    PlayMode:EUMGSequencePlayMode Enum
    PlaybackSpeed:float=2
   ) -> (then:Exec, Result:UMGSequence Player Object Reference, Finished:Exec) {is=title K2Node_PlayAnimation}
   > SetActivePip!(
    execute
    self:WBP Compete Notification Pip List Object Reference=Get PipList() -> (PipList:WBP Compete Notification Pip List Object Reference)
    Index:int=Get CurrentFrameIndex() -> (CurrentFrameIndex:int=0)
   ) -> (then)
   UpdateMainText(execute=@n10.Finished)
   > n11 = Branch(Condition=Is Timer Paused by Handle!(Handle:Timer Handle Structure=Get TimerHandle() -> (TimerHandle:Timer Handle Structure)) -> (ReturnValue:bool))
   StartTimer!(execute=@n11.else, self, Time:double=Get Time() -> (Time:double)) -> (then) {is=title K2Node_CallFunction}
   Pause Timer by Handle!(execute=@n9.else, Handle:Timer Handle Structure=Get TimerHandle()) -> (then)
   Event Construct()
   > n12 = Sequence()
   > Parent: Construct()
   > SetNavigationRuleCustom!(execute, self:Widget Object Reference, Direction:EUINavigation Enum=Right, InCustomDelegate:Delegate=Create Event()) -> (then)
   > n13 = Is Valid(InputObject=Get FortTournamentBookmarksVM())
   n14 = ConstructVM(execute=@n13.Is Not Valid) {is=title K2Node_CallFunction}
   n15 = Set Timer by Event(execute=[@n14.then, @n13.Is Valid], Delegate=Create Event(), Time=1, bLooping=true, bMaxOncePerFrame=true)
   > Set TickCountdownTimerHandle(TickCountdownTimerHandle:Timer Handle Structure=@n15) -> (Output_Get:Timer Handle Structure)
   SetUpForMobile(execute=@n12.then_1)
   Event On Hovered()
   > Parent: On Hovered() {is=title K2Node_CallParentFunction}
   > n16 = Branch(Condition=Get Live() -> (Live:bool))
   UpdateMainText(execute=@n16.else)
   > UpdateCountdownText(ShowMinTime=false)
   > Branch(Condition=Get bMultipleBookmarked())
   > StartTimer(Time=Get Time())
   Event On Unhovered()
   > Parent: On Unhovered() {is=title K2Node_CallParentFunction}
   > Pause Timer by Handle(Handle=Get TimerHandle())
   > n17 = Branch(Condition=Get Live())
   UpdateCountdownText(execute=@n17.else, ShowMinTime=true)
   Event On Clicked()
   > Parent: On Clicked() {is=title K2Node_CallParentFunction}
   n18 = Get FortTournamentBookmarksVM()
   > Is Valid(InputObject=@n18)
   n19 = Get Bookmarked Events()
   > Branch(Condition=Is Not Empty(TargetArray=@n19))
   n20 = "Get (a ref)"(Array:Array of Fort Upcoming Tournament Data Structures=@n19, Dimension 1=Get CurrentFrameIndex() -> (CurrentFrameIndex=0)) -> (Output:"Fort Upcoming Tournament Data Structure (by ref)")
   n21 = Break Fort Upcoming Tournament Data!(FortUpcomingTournamentData:"Fort Upcoming Tournament Data Structure (by ref)"=@n20) -> (Title:Text, EventId:String, EventWindowId:String, BeginTime:Date Time Structure, EndTime:Date Time Structure) {is=title K2Node_BreakStruct}
   > n22 = Branch(Condition=IsEmpty(InString=@n21.EventId))
   n23 = Branch(
    execute=@n22.else
    Condition=IsUpcomingTournamentLive!(self:Tournament Bookmarks VM Object Reference=@n18, UpcomingTournament:"Fort Upcoming Tournament Data Structure (by ref)"=@n20) -> (ReturnValue:bool)
   )
   > GoToLobbyWithTournament!(execute, self:Tournament Bookmarks VM Object Reference=@n18, UpcomingTournament:"Fort Upcoming Tournament Data Structure (by ref)"=@n20) -> (then)
   ShowTournamentDetailsScreen!(
    execute=@n23.else
    self:Fort Compete UIManager Object Reference=Get FortCompeteUIManager()
    Params_ContextPC:Player Controller Object Reference=GetOwningPlayer()
    Params_TournamentId:String=@n21.EventId
    Params_TournamentWindowId:String=@n21.EventWindowId
   ) -> (then)
   Event PreConstruct()
   > SetMobileStyle() {is=title K2Node_CallFunction}
   Event Destruct()
   > Clear and Invalidate Timer by Handle(Handle=Get TickCountdownTimerHandle() -> (TickCountdownTimerHandle:Timer Handle Structure))
  ]}
 GetFormatedExpirationText {nodes [
   n1 = GetFormatedExpirationText!() -> (then, Timespan:Timespan Structure) {is=title K2Node_FunctionEntry}
   > n2 = Branch(Condition="Timespan > Timespan"!(A:Timespan Structure=@n1, B_Days:int=0, B_Hours:int=0, B_Minutes:int=0, B_Seconds:int=0, B_Milliseconds:int=0) -> (ReturnValue:bool=~) {is=title K2Node_PromotableOperator})
   > SetTimeFontSize!(execute, self, Font Size:float=Get SmallFontSize() -> (SmallFontSize:float)) -> (then)
   > SetHorizontalAlignment(
    self:Horizontal Box Slot Object Reference=SlotAsHorizontalBoxSlot!(Widget:Widget Object Reference=Get Timer() -> (Timer:Common Text Object Reference)) -> (ReturnValue:Horizontal Box Slot Object Reference)
    InHorizontalAlignment=HAlign_Center
   )
   n3 = Get!() -> (Timespan_Days:int=0, Timespan_Hours:int=0, Timespan_Minutes:int=0, Timespan_Seconds:int=0, Timespan_Milliseconds:int=0)
   > Set Weeks(Weeks:int=int / int(A=@n3, B=7)) -> (Output_Get:int=0)
   > Set Days(Days:int="% (Integer)"(A=@n3, B=7) -> (ReturnValue=0)) -> (Output_Get:int=0)
   > n4 = Branch(Condition="integer >= integer"(A=Get Weeks() -> (Weeks:int=0), B=1))
   > Return Node!(
    execute
    ExpirationText:Text=Format Text!(Format="LOCTEXT(\"{Weeks}W {Days}D\")", Weeks:int=Get Weeks() -> (Weeks=0), Days:int=Get Days() -> (Days:int=0)) -> (Result)
   )
   n5 = Branch(execute=@n4.else, Condition="integer > integer"(A=Get Days() -> (Days=0)))
   > Return Node(ExpirationText=Format Text!(Format="LOCTEXT(\"{Days}D\")", Days=Get Days() -> (Days=0)) -> (Result))
   n6 = Branch(
    execute=@n5.else
    Condition="integer > integer"(A=Get() -> (Timespan_Days=0, Timespan_Hours=0, Timespan_Minutes=0, Timespan_Seconds=0, Timespan_Milliseconds=0).Timespan_Hours)
   )
   > Return Node(ExpirationText=Format Text!(
     Format="LOCTEXT(\"{Hours}H\")"
     Hours:int=Get() -> (Timespan_Days=0, Timespan_Hours=0, Timespan_Minutes=0, Timespan_Seconds=0, Timespan_Milliseconds=0).Timespan_Hours
    ) -> (Result))
   SetTimeFontSize(execute=@n6.else, Font Size=Get BigFontSize() -> (BigFontSize:float))
   > SetHorizontalAlignment(self=SlotAsHorizontalBoxSlot(Widget=Get Timer()), InHorizontalAlignment=HAlign_Left)
   n7 = Get() -> (Timespan_Days=0, Timespan_Hours=0, Timespan_Minutes=0, Timespan_Seconds=0, Timespan_Milliseconds=0)
   > n8 = Branch(Condition="integer > integer"(A=@n7.Timespan_Minutes))
   n9 = Get() -> (Timespan_Days=0, Timespan_Hours=0, Timespan_Minutes=0, Timespan_Seconds=0, Timespan_Milliseconds=0)
   > Return Node(ExpirationText=Select(
     Option 0:Text=Format Text!(Format="LOCTEXT(\"{Minutes}:{Seconds}\")", Minutes:int=@n9.Timespan_Minutes, Seconds:int=@n9.Timespan_Seconds) -> (Result)
     Option 1:Text=Format Text(Format="LOCTEXT(\"{Minutes}:0{Seconds}\")", Minutes=@n9.Timespan_Minutes, Seconds=@n9.Timespan_Seconds)
     Index="integer < integer"(A=@n9.Timespan_Seconds, B=10)
    ) -> (ReturnValue:Text))
   Return Node(execute=@n8.else, ExpirationText=Format Text!(Format="LOCTEXT(\"{Seconds}s\")", Seconds=@n7.Timespan_Seconds) -> (Result))
   Return Node(execute=@n2.else)
   Get() -> (Timespan_Days=0, Timespan_Hours=0, Timespan_Minutes=0, Timespan_Seconds=0, Timespan_Milliseconds=0)
  ]}
 GetLowestCountdownIndex {nodes [
   GetLowestCountdownIndex() {is=title K2Node_FunctionEntry}
   > n1 = Branch(Condition="integer > integer"(A=Get MinCountdownIndex() -> (MinCountdownIndex:int=0)))
   > Return Node!(execute, MinCountdownIndex:int=Get MinCountdownIndex() -> (MinCountdownIndex=0))
   n2 = Get CurrentFrameIndex() -> (CurrentFrameIndex=0)
   Set MinCountdownIndex(execute=@n1.else, MinCountdownIndex:int=@n2) -> (Output_Get:int=0)
   n3 = Get Bookmarked Events()
   > Set MinTime(MinTime:Timespan Structure=GetRemainingTimeToStart!(
     self:Tournament Bookmarks VM Object Reference=Get FortTournamentBookmarksVM()
     UpcomingTournament:"Fort Upcoming Tournament Data Structure (by ref)"="Get (a ref)"(Array=@n3, Dimension 1=@n2)
    ) -> (ReturnValue:Timespan Structure)) -> (Output_Get:Timespan Structure)
   > n4 = For Each Loop(Array:Array of Fort Upcoming Tournament Data Structures=Get Bookmarked Events()) -> (Array Element:Fort Upcoming Tournament Data Structure)
   Return Node(execute=@n4.Completed, MinCountdownIndex=Get MinCountdownIndex() -> (MinCountdownIndex=0))
   n5 = GetRemainingTimeToStart(self=Get FortTournamentBookmarksVM(), UpcomingTournament=@n4)
   > Branch(Condition="Timespan > Timespan"!(A=Get MinTime() -> (MinTime:Timespan Structure), B:Timespan Structure=@n5) -> (ReturnValue=~))
   > Set MinTime(MinTime=@n5)
   > Set MinCountdownIndex(MinCountdownIndex=@n4.Array Index) -> (Output_Get=0)
  ]}
 IncrementIndex {nodes [
   n1 = IncrementIndex!() -> (then, PreviousFrame:bool=~) {is=title K2Node_FunctionEntry}
   > n2 = Branch(Condition=@n1)
   > n3 = Branch(Condition="integer > integer"(A=Get CurrentFrameIndex() -> (CurrentFrameIndex=0)))
   > n4 = Decrement Int(Value=Get CurrentFrameIndex() -> (CurrentFrameIndex=0))
   > Return Node!(execute, NewIndex:int=@n4)
   n5 = Set CurrentFrameIndex(execute=@n3.else, CurrentFrameIndex:int=int - int(A=Get BookmarkedEventsAmount() -> (BookmarkedEventsAmount:int=0))) -> (Output_Get:int=0)
   > Return Node(NewIndex=@n5)
   n6 = Get BookmarkedEventsAmount() -> (BookmarkedEventsAmount=0)
   n7 = Branch(execute=@n2.else, Condition="integer < integer"(A=Get CurrentFrameIndex() -> (CurrentFrameIndex=0), B=int - int(A=@n6)))
   > n8 = Increment Int(Value=Get CurrentFrameIndex() -> (CurrentFrameIndex=0))
   > Return Node(NewIndex=@n8)
   n9 = Set CurrentFrameIndex(execute=@n7.else, CurrentFrameIndex=0) -> (Output_Get=0)
   > Return Node(NewIndex=@n9)
  ]}
 SetBookmarkedTournamentsAmount {nodes [
   n1 = SetBookmarkedTournamentsAmount!() -> (then, Amount:int) {is=title K2Node_FunctionEntry}
   > SetVisibility(
    self=Get Overlay_Bookmarks() -> (Overlay_Bookmarks:Overlay Object Reference)
    InVisibility=Select(Option 0:ESlateVisibility Enum=Collapsed, Option 1:ESlateVisibility Enum=SelfHitTestInvisible, Index="integer > integer"(A=@n1, B=1)) -> (ReturnValue:ESlateVisibility Enum)
   )
   > "SetText (Text)"(
    self=Get BookmarkedAmount() -> (BookmarkedAmount:Common Text Object Reference)
    InText="To Text (Integer)"(Value=Get Amount() -> (Amount:int=0))
   )
  ]}
 SetBookmarkedUpcomingTournaments {nodes [
   n1 = SetBookmarkedUpcomingTournaments!() -> (then, BookmarkedEvents:Array of Fort Upcoming Tournament Data Structures) {is=title K2Node_FunctionEntry}
   > n2 = Set Bookmarked Events(Bookmarked Events:Array of Fort Upcoming Tournament Data Structures=@n1) -> (Output_Get:Array of Fort Upcoming Tournament Data Structures)
   > n3 = Sequence!(execute) -> (then_0, then_1, then_2:Exec, then_3:Exec)
   n4 = Is Empty!(TargetArray:Array of Fort Upcoming Tournament Data Structures=@n1) -> (ReturnValue:bool)
   > n5 = Branch(Condition=@n4)
   n6 = Length(TargetArray:Array of Fort Upcoming Tournament Data Structures=@n2)
   Set bMultipleBookmarked(execute=@n5.else, bMultipleBookmarked:bool="integer > integer"(A=@n6, B=1)) -> (Output_Get:bool)
   > Set BookmarkedEventsAmount(BookmarkedEventsAmount:int=@n6) -> (Output_Get:int=0)
   > Set CurrentFrameIndex(CurrentFrameIndex=0) -> (Output_Get=0)
   > SetVisibility(self=Get PipList(), InVisibility=Select(Option 0=Collapsed, Option 1=SelfHitTestInvisible, Index=Get bMultipleBookmarked()))
   > SetPipsCount!(
    execute
    self:WBP Compete Notification Pip List Object Reference=Get PipList()
    Count:int=Get BookmarkedEventsAmount() -> (BookmarkedEventsAmount=0)
    Duration Value:double=Get Time()
    Anim Pip:bool=true
    Show Progress:bool=false
   ) -> (then)
   > SetActivePip(self=Get PipList(), Index=0)
   > UpdateCountdownText(ShowMinTime=true)
   > SetBookmarkedTournamentsAmount!(execute, self, Amount:int=Length(TargetArray=Get Bookmarked Events())) -> (then) {is=title K2Node_CallFunction}
   > Branch(Condition=IsMobileGame())
   > UpdateAnimText()
   > UpdateMainText()
   > UpdateCountdownText(ShowMinTime=false)
   > StartTimer(Time=Get Time())
   UpdateVisibility(execute=@n3.then_1)
   n7 = Branch(execute=@n3.then_2, Condition=@n4)
   n8 = Get FortTournamentBookmarksVM()
   n9 = Is Valid(exec=@n7.else, InputObject=@n8)
   n10 = UpdateLiveState(execute=@n9.Is Not Valid)
   n11 = "Get (a ref)"(Array=Get Bookmarked Events(), Dimension 1=0)
   SetLifeTournamentName!(execute=[@n10.then, @n12.then], self, InText:Text=Break Fort Upcoming Tournament Data(FortUpcomingTournamentData=@n11)) -> (then)
   n12 = SetLiveState!(execute=@n9.Is Valid, self, InLiveState:bool=IsUpcomingTournamentLive(self=@n8, UpcomingTournament=@n11)) -> (then)
  ]}
 SetLifeTournamentName {nodes [
   n1 = SetLifeTournamentName!() -> (then, InText:Text) {is=title K2Node_FunctionEntry}
   > "SetText (Text)"(self=Get TournamentName_Live() -> (TournamentName_Live:Common Text Object Reference), InText=@n1)
  ]}
 SetLiveState {nodes [
   n1 = SetLiveState!() -> (then, InLiveState:bool=~) {is=title K2Node_FunctionEntry}
   > Set LocalLiveChanged(LocalLiveChanged:bool="Not Equal (Boolean)"(A=Get Live(), B=@n1) -> (ReturnValue=~)) -> (Output_Get:bool)
   n2 = Get InLiveState() -> (InLiveState:bool)
   > Set Live(Live:bool=@n2) -> (Output_Get:bool)
   > SetActiveWidgetIndex(
    self:Common Visibility Switcher Object Reference=[
     Get Switcher_Content() -> (Switcher_Content:Common Visibility Switcher Object Reference)
     Get TooltipSwitcher() -> (TooltipSwitcher:Common Visibility Switcher Object Reference)
    ]
    Index=Select(Option 0:int=0, Option 1:int=1, Index=@n2) -> (ReturnValue:int)
   )
   > SetVisibility(
    self=Get Image_Live() -> (Image_Live:Image Object Reference)
    InVisibility=Select(Option 0:ESlateVisibility Enum=Collapsed, Option 1:ESlateVisibility Enum=SelfHitTestInvisible, Index=@n2) -> (ReturnValue:ESlateVisibility Enum)
   )
  ]}
 SetMobileStyle {nodes [
   SetMobileStyle() {is=title K2Node_FunctionEntry}
   > Branch(Condition=IsMobileGame())
   n1 = Get MobilePadding() -> (MobilePadding:Margin Structure)
   > SetPadding!(
    execute
    self:Overlay Slot Object Reference=SlotAsOverlaySlot(Widget=Get TooltipRoot() -> (TooltipRoot:Grid Panel Object Reference))
    InPadding:Margin Structure=@n1
   ) -> (then)
   > SetPadding(self=SlotAsOverlaySlot(Widget=Get TooltipRoot_Live() -> (TooltipRoot_Live:Grid Panel Object Reference)), InPadding=@n1)
  ]}
 SetTimeFontSize {nodes [
   n1 = SetTimeFontSize!() -> (then, Font Size:float=~) {is=title K2Node_FunctionEntry}
   n2 = Get Timer()
   > n3 = Set members in Slate Font Info!(
    execute:Exec
    StructRef:"Slate Font Info Structure (by ref)"=Get Font!(self:Text Object Reference=@n2) -> (Font:Slate Font Info Structure)
    Size:float=float * float(A=@n1, B=0.75) {is=title K2Node_PromotableOperator}
   ) -> (then:Exec, StructOut:"Slate Font Info Structure (by ref)") {is=title K2Node_SetFieldsInStruct}
   > SetFont(self=@n2, InFontInfo=@n3)
  ]}
 SetTournamentName {nodes [
   n1 = SetTournamentName!() -> (then, TournamentName:Text) {is=title K2Node_FunctionEntry}
   > "SetText (Text)"(self=Get TournamentName_Bookmarked() -> (TournamentName_Bookmarked:Common Text Object Reference), InText=@n1)
   > Set RenderOpacity!(execute, RenderOpacity:float=1, self:Widget Object Reference=Get TournamentName_Bookmarked()) -> (then, Output_Get:float)
   > Set RenderOpacity(RenderOpacity=0, self=Get TournamentName_Bookmarked_ForAnim() -> (TournamentName_Bookmarked_ForAnim:Common Text Object Reference))
  ]}
 SetUpForMobile {nodes [
   SetUpForMobile() {is=title K2Node_FunctionEntry}
   > n1 = Set CollapseException(CollapseException:bool=IsMobileGame()) -> (Output_Get:bool)
   > Branch(Condition=@n1)
   > ExpandExtPart(self:WBP Expandable Base Object Reference)
   > RequestTransition!(
    execute
    self:UIKit Block Object Reference=Get Block_Outline() -> (Block_Outline:WBP UIKit Block Outline Object Reference)
    State:EUIKitBlockVisualState Enum=Unfocused
   ) -> (then)
   > SetVisibility(self=Get Highlight() -> (Highlight:WBP UIKit Highlight Object Reference), InVisibility=Collapsed)
  ]}
 SkipCurrentFrame {nodes [
   SkipCurrentFrame!() -> (then, Navigation:EUINavigation Enum) {is=title K2Node_FunctionEntry}
   > Clear and Invalidate Timer by Handle(Handle=Get TimerHandle())
   > ChangeFrame(bPreviousFrame=false) {is=title K2Node_CallFunction}
   > Return Node!(execute, ReturnValue:Widget Object Reference=Self-Reference())
  ]}
 UpdateAnimText {nodes [
   UpdateAnimText() {is=title K2Node_FunctionEntry}
   > "SetText (Text)"(
    self=Get TournamentName_Bookmarked_ForAnim()
    InText="Get (a ref)"!(Array=Get Bookmarked Events(), Dimension 1=Get CurrentFrameIndex() -> (CurrentFrameIndex=0)) -> (
     Output_Title:Text
     Output_EventId:String
     Output_EventWindowId:String
     Output_BeginTime:Date Time Structure
     Output_EndTime:Date Time Structure
    )
   )
  ]}
 UpdateCountdownText {nodes [
   UpdateCountdownText!() -> (then, ShowMinTime:bool=~) {is=title K2Node_FunctionEntry}
   > Is Valid(InputObject=Get FortTournamentBookmarksVM())
   > n1 = Sequence!(execute) -> (then_0, then_1)
   > SetTimeFontSize(Font Size=Get BigFontSize()) {is=title K2Node_CallFunction}
   n2 = Get ShowMinTime() -> (ShowMinTime:bool)
   n3 = Get bMultipleBookmarked()
   n4 = "Get (a ref)"!(
    Array=Get Bookmarked Events()
    Dimension 1=Select(
     Option 0:int=Get CurrentFrameIndex() -> (CurrentFrameIndex=0)
     Option 1:int=GetLowestCountdownIndex!() -> (MinCountdownIndex:int=0) {is=title K2Node_CallFunction}
     Index=AND Boolean(A=@n2, B=@n3)
    ) -> (ReturnValue:int)
   ) -> (Output:"Fort Upcoming Tournament Data Structure (by ref)")
   n5 = Get FortTournamentBookmarksVM()
   "SetText (Text)"(
    execute=@n1.then_1
    self=Get Timer()
    InText=GetFormatedExpirationText!(Timespan:Timespan Structure=GetRemainingTimeToStart(self=@n5, UpcomingTournament=@n4)) -> (ExpirationText:Text) {is=title K2Node_CallFunction}
   )
  ]}
 UpdateIsTournamentLobby {nodes [
   n1 = UpdateIsTournamentLobby!() -> (then, bIsTournamentLobby:bool=~) {is=title K2Node_FunctionEntry}
   > Set Is Tournament Lobby(Is Tournament Lobby:bool=@n1) -> (Output_Get:bool)
   > UpdateVisibility()
  ]}
 UpdateLiveState {nodes [
   UpdateLiveState() {is=title K2Node_FunctionEntry}
   n1 = Get FortTournamentBookmarksVM()
   > Is Valid(InputObject=@n1)
   n2 = Get Bookmarked Events()
   > Branch(Condition=Is Not Empty(TargetArray=@n2))
   n3 = "Get (a ref)"(Array=@n2, Dimension 1=0)
   n4 = IsUpcomingTournamentLive(self=@n1, UpcomingTournament=@n3)
   > Branch(Condition="Not Equal (Boolean)"(A=Get Live(), B=@n4) -> (ReturnValue=~))
   > SetLiveState(InLiveState=@n4) {is=title K2Node_CallFunction}
   > Set CurrentFrameIndex(CurrentFrameIndex=0) -> (Output_Get=0)
   > Branch(Condition=@n4)
   > SetLifeTournamentName(InText=Break Fort Upcoming Tournament Data(FortUpcomingTournamentData=@n3)) {is=title K2Node_CallFunction}
   > ShowFrontendBookmarkedTournamentNotification!(
    execute
    self:Fort Compete UIManager Object Reference=Get FortCompeteUIManager()
    NotificationClass:Fort UINotification Class Reference=&/BP_CompeteBookmarkedTournamentLiveNotification.BP_CompeteBookmarkedTournamentLiveNotification_C
    UpcomingTournament:"Fort Upcoming Tournament Data Structure (by ref)"=@n3
   ) -> (then)
  ]}
 UpdateMainText {nodes [
   UpdateMainText() {is=title K2Node_FunctionEntry}
   > "SetText (Text)"(
    self=Get TournamentName_Bookmarked()
    InText="Get (a ref)"!(Array=Get Bookmarked Events(), Dimension 1=Get CurrentFrameIndex() -> (CurrentFrameIndex=0)) -> (
     Output_Title:Text
     Output_EventId:String
     Output_EventWindowId:String
     Output_BeginTime:Date Time Structure
     Output_EndTime:Date Time Structure
    )
   )
   > Set RenderOpacity(RenderOpacity=1, self=Get TournamentName_Bookmarked())
   > Set RenderOpacity(RenderOpacity=0, self=Get TournamentName_Bookmarked_ForAnim())
  ]}
 UpdateVisibility {nodes [
   UpdateVisibility() {is=title K2Node_FunctionEntry}
   n1 = NOT Boolean(A=IsMMWidgetEnabled!(self:Tournament Bookmarks VM Object Reference=Get FortTournamentBookmarksVM()) -> (ReturnValue:bool))
   n2 = Get Is Tournament Lobby() -> (Is Tournament Lobby:bool)
   n3 = AND Boolean(A=Is Not Empty(TargetArray=Get Bookmarked Events()), B=NOT Boolean(A=AND Boolean(A=@n2, B=@n1)))
   > SetVisibility(InVisibility=Select(Option 0:ESlateVisibility Enum=Collapsed, Option 1:ESlateVisibility Enum=SelfHitTestInvisible, Index=@n3) -> (ReturnValue:ESlateVisibility Enum))
   > n4 = Branch(Condition=@n3)
   n5 = Get TickCountdownTimerHandle()
   > Unpause Timer by Handle!(execute, Handle:Timer Handle Structure=@n5) -> (then)
   Pause Timer by Handle(execute=@n4.else, Handle=@n5)
  ]}
]
variables [
 BigFontSize {type=number, default=15}
 bMultipleBookmarked {type=boolean}
 Bookmarked Events {type="array<FortUpcomingTournamentData>", default []}
 BookmarkedEventsAmount {type=integer}
 CurrentFrameIndex {type=integer}
 Is Ranked {type=boolean}
 Is Tournament Lobby {type=boolean}
 Live {type=boolean}
 MinCountdownIndex {type=integer, default=-1}
 MobilePadding {type=Margin, default=<ltrb 16 0 12 0>}
 ShowOnlyOnRankedLobbies {type=boolean}
 SmallFontSize {type=number, default=15}
 TickCountdownTimerHandle {type=TimerHandle}
 Time {type=number, default=2}
 TimerHandle {type=TimerHandle}
]
widgets {
 info {
  inheritedWidgetCount=0
  namedSlotCount=3
  parentClass=/Game/UI/Frontend/Lobby/Habanero/WBP_Expandable_Base.WBP_Expandable_Base_C
  rootWidgetClass=null
  widgetCount=26
 }
 widgets [
  ContentRoot {widgetClassPath=&GridPanel, widget=ContentRoot, properties {columnFill [1], rowFill [0,1]}}
  GridPanel_2 {
   widgetClassPath=&GridPanel
   parent=ContentRoot
   widget=GridPanel_2
   properties {columnFill [0,1]}
   slot {class=GridSlot, properties {horizontalAlignment=HAlign_Center, layer=1, verticalAlignment=VAlign_Top, padding=<ltrb 0 -8 0 0>}}
  }
  Image_Cup {
   widgetClassPath=&Image
   bIsVariable=true
   parent=GridPanel_2
   widget=Image_Cup
   properties {brush {
     bIsDynamicallyLoaded=false
     drawAs=Image
     imageType=FullColor
     mirroring=NoMirror
     resourceName=None
     resourceObject=/FortUILibrary/Textures/MaterialInstances/MI_UIKit_Icon_Compete
     tiling=NoTile
     imageSize=<38 38>
     margin=<ltrb 0 0 0 0>
     outlineSettings {
      bUseBrushTransparency=false
      roundingType=HalfHeightRadius
      width=0
      color {colorUseRule=UseColor_Specified, specifiedColor=<rgba 0 0 0 0>}
      cornerRadii=<0 0 0 0>
     }
     tintColor {colorUseRule=UseColor_Specified, specifiedColor=<rgba 0.013 0.013 0.013 1>}
     uVRegion {bIsValid=false, max=<0 0>, min=<0 0>}
    }}
   slot {class=GridSlot, properties {horizontalAlignment=HAlign_Center, layer=5, verticalAlignment=VAlign_Center}}
  }
  CupBackground {
   widgetClassPath=&Image
   bIsVariable=true
   parent=GridPanel_2
   widget=CupBackground
   properties {brush {
     bIsDynamicallyLoaded=false
     drawAs=Image
     imageType=FullColor
     mirroring=NoMirror
     resourceName=None
     resourceObject=/Compete/MaterialInstances/MI_UI_Compete_Flag
     tiling=NoTile
     imageSize=<47 66>
     margin=<ltrb 0 0 0 0>
     outlineSettings {
      bUseBrushTransparency=false
      roundingType=HalfHeightRadius
      width=0
      color {colorUseRule=UseColor_Specified, specifiedColor=<rgba 0 0 0 0>}
      cornerRadii=<0 0 0 0>
     }
     tintColor {colorUseRule=UseColor_Specified, specifiedColor=<rgba 1 1 1 1>}
     uVRegion {bIsValid=false, max=<0 0>, min=<0 0>}
    }}
   slot {class=GridSlot}
  }
  Switcher_Content {
   widgetClassPath=&CommonVisibilitySwitcher
   bIsVariable=true
   parent=ContentRoot
   widget=Switcher_Content
   slot {class=GridSlot, properties {horizontalAlignment=HAlign_Center, row=1, verticalAlignment=VAlign_Bottom}}
  }
  Overlay_Timer {
   widgetClassPath=&Overlay
   bIsVariable=true
   parent=Switcher_Content
   widget=Overlay_Timer
   slot {class=CommonVisibilitySwitcherSlot}
  }
  TimerBackplate {
   widgetClassPath=&WBP_UIKit_Backplate
   parent=Overlay_Timer
   widget=TimerBackplate
   properties {brightness=Dark, cornerRadius=Circle, backplateSize=<72 24>}
   slot {class=OverlaySlot, properties {horizontalAlignment=HAlign_Fill, verticalAlignment=VAlign_Fill}}
  }
  HorizontalBox_0 {
   widgetClassPath=&HorizontalBox
   parent=Overlay_Timer
   widget=HorizontalBox_0
   slot {class=OverlaySlot, properties {horizontalAlignment=HAlign_Fill, verticalAlignment=VAlign_Center, padding=<ltrb 8 0 8 0>}}
  }
  Image_Clock {
   widgetClassPath=&Image
   bIsVariable=true
   parent=HorizontalBox_0
   widget=Image_Clock
   properties {brush {
     bIsDynamicallyLoaded=false
     drawAs=Image
     imageType=FullColor
     mirroring=NoMirror
     resourceName=None
     resourceObject=/FortUILibrary/Textures/MaterialInstances/MI_UIKit_Icon_UpdatePending
     tiling=NoTile
     imageSize=<12 12>
     margin=<ltrb 0 0 0 0>
     outlineSettings {
      bUseBrushTransparency=false
      roundingType=HalfHeightRadius
      width=0
      color {colorUseRule=UseColor_Specified, specifiedColor=<rgba 0 0 0 0>}
      cornerRadii=<0 0 0 0>
     }
     tintColor {colorUseRule=UseColor_Specified, specifiedColor=<rgba 1 1 1 1>}
     uVRegion {bIsValid=false, max=<0 0>, min=<0 0>}
    }}
   slot {class=HorizontalBoxSlot, properties {horizontalAlignment=HAlign_Left, verticalAlignment=VAlign_Center, padding=<ltrb 0 0 4 0>}}
  }
  Timer {
   widgetClassPath=&CommonTextBlock
   bIsVariable=true
   parent=HorizontalBox_0
   widget=Timer
   properties {
    text=11:11
    visibility=Visible
    font {
     bForceMonospaced=false
     bMaterialIsStencil=false
     fontMaterial=null
     fontObject=&HeadingNow
     letterSpacing=0
     monospacedWidth=1
     size=11.25
     skewAmount=0
     typefaceFontName=74 Regular
     outlineSettings {
      bApplyOutlineToDropShadows=false
      bMiteredCorners=false
      bSeparateFillAlpha=false
      outlineMaterial=null
      outlineSize=0
      outlineColor=<rgba 0 0 0 1>
     }
    }
    shadowOffset=<0 0>
   }
   slot {class=HorizontalBoxSlot, properties {horizontalAlignment=HAlign_Left, size {sizeRule=Fill, value=1}}}
  }
  Highlight {
   widgetClassPath=/FortUILibrary/UIKit/Highlight/WBP_UIKit_Highlight
   bIsVariable=true
   parent=Switcher_Content
   widget=Highlight
   properties {highlightSize=Compact, highlightText=Live}
   slot {class=CommonVisibilitySwitcherSlot, properties {horizontalAlignment=HAlign_Center, verticalAlignment=VAlign_Center}}
  }
  Overlay_Bookmarks {
   widgetClassPath=&Overlay
   bIsVariable=true
   parent=ContentRoot
   widget=Overlay_Bookmarks
   slot {class=GridSlot, properties {horizontalAlignment=HAlign_Right, layer=2, verticalAlignment=VAlign_Top, padding=<ltrb 4 0 0 0>}}
  }
  BookmarksBackplate {
   widgetClassPath=&WBP_UIKit_Backplate
   parent=Overlay_Bookmarks
   widget=BookmarksBackplate
   properties {brightness=Dark, cornerRadius=Circle, backplateSize=<26 24>}
   slot {class=OverlaySlot, properties {horizontalAlignment=HAlign_Fill, verticalAlignment=VAlign_Fill}}
  }
  BookmarkedAmount {
   widgetClassPath=&CommonTextBlock
   bIsVariable=true
   parent=Overlay_Bookmarks
   widget=BookmarkedAmount
   properties {
    text=8
    font {
     bForceMonospaced=false
     bMaterialIsStencil=false
     fontMaterial=null
     fontObject=&HeadingNow
     letterSpacing=0
     monospacedWidth=1
     size=12.75
     skewAmount=0
     typefaceFontName=86 Bold Italic
     outlineSettings {
      bApplyOutlineToDropShadows=false
      bMiteredCorners=false
      bSeparateFillAlpha=false
      outlineMaterial=null
      outlineSize=0
      outlineColor=<rgba 0 0 0 1>
     }
    }
    shadowOffset=<0 0>
   }
   slot {class=OverlaySlot, properties {horizontalAlignment=HAlign_Center, verticalAlignment=VAlign_Center, padding=<ltrb 5 0 5 0>}}
  }
  SizeBox_2 {
   widgetClassPath=&SizeBox
   widget=SizeBox_2
   properties {bOverride_HeightOverride=true, bOverride_WidthOverride=true, heightOverride=89, minDesiredHeight=89, widthOverride=349}
  }
  GridPanel_99 {
   widgetClassPath=&GridPanel
   parent=SizeBox_2
   widget=GridPanel_99
   properties {columnFill [1], rowFill [1]}
   slot {class=SizeBoxSlot, properties {padding=<ltrb 32 0 10 0>}}
  }
  PipList {
   widgetClassPath=&/WBP_CompeteNotification_Pip_List
   bIsVariable=true
   parent=GridPanel_99
   widget=PipList
   slot {class=GridSlot, properties {row=1}}
  }
  TooltipSwitcher {
   widgetClassPath=&CommonVisibilitySwitcher
   bIsVariable=true
   parent=GridPanel_99
   widget=TooltipSwitcher
   slot {class=GridSlot}
  }
  TooltipRoot {
   widgetClassPath=&GridPanel
   bIsVariable=true
   parent=TooltipSwitcher
   widget=TooltipRoot
   properties {rowFill [1,0]}
   slot {class=CommonVisibilitySwitcherSlot}
  }
  TournamentName_Bookmarked {
   widgetClassPath=&CommonTextBlock
   bIsVariable=true
   parent=TooltipRoot
   widget=TournamentName_Bookmarked
   properties {
    bIsScrollingEnabled=false
    lineHeightPercentage=0.9
    minDesiredWidth=290
    mobileFontSizeMultiplier=0.94
    scrollStyle=&TextScrollStyle-Base_C
    text="Bookmarked tournament "
    textOverflowPolicy=MultilineEllipsis
    textTransformPolicy=ToUpper
    wrapTextAt=290
    colorAndOpacity {colorUseRule=UseColor_Specified, specifiedColor=<rgba 0.0048 0.004 0.0116 0.702>}
    font {
     bForceMonospaced=false
     bMaterialIsStencil=false
     fontMaterial=null
     fontObject=&HeadingNow
     letterSpacing=10
     monospacedWidth=1
     size=15.75
     skewAmount=0
     typefaceFontName=76 Bold
     outlineSettings {
      bApplyOutlineToDropShadows=false
      bMiteredCorners=false
      bSeparateFillAlpha=false
      outlineMaterial=null
      outlineSize=0
      outlineColor=<rgba 0 0 0 1>
     }
    }
    shadowOffset=<0 0>
   }
   slot {class=GridSlot, properties {horizontalAlignment=HAlign_Left, verticalAlignment=VAlign_Center}}
  }
  TournamentName_Bookmarked_ForAnim {
   widgetClassPath=&CommonTextBlock
   bIsVariable=true
   parent=TooltipRoot
   widget=TournamentName_Bookmarked_ForAnim
   properties {
    bIsScrollingEnabled=false
    lineHeightPercentage=0.9
    minDesiredWidth=290
    mobileFontSizeMultiplier=0.94
    renderOpacity=0
    scrollStyle=&TextScrollStyle-Base_C
    text=Bookmarked tournament
    textOverflowPolicy=MultilineEllipsis
    textTransformPolicy=ToUpper
    wrapTextAt=290
    colorAndOpacity {colorUseRule=UseColor_Specified, specifiedColor=<rgba 0.0048 0.004 0.0116 0.702>}
    font {
     bForceMonospaced=false
     bMaterialIsStencil=false
     fontMaterial=null
     fontObject=&HeadingNow
     letterSpacing=10
     monospacedWidth=1
     size=15.75
     skewAmount=0
     typefaceFontName=76 Bold
     outlineSettings {
      bApplyOutlineToDropShadows=false
      bMiteredCorners=false
      bSeparateFillAlpha=false
      outlineMaterial=null
      outlineSize=0
      outlineColor=<rgba 0 0 0 1>
     }
    }
    shadowOffset=<0 0>
   }
   slot {class=GridSlot, properties {horizontalAlignment=HAlign_Left, verticalAlignment=VAlign_Center}}
  }
  TooltipRoot_Live {
   widgetClassPath=&GridPanel
   bIsVariable=true
   parent=TooltipSwitcher
   widget=TooltipRoot_Live
   properties {columnFill [1], rowFill [0,1]}
   slot {class=CommonVisibilitySwitcherSlot, properties {verticalAlignment=VAlign_Center}}
  }
  JoinNowText {
   widgetClassPath=&CommonTextBlock
   parent=TooltipRoot_Live
   widget=JoinNowText
   properties {
    minDesiredWidth=290
    mobileFontSizeMultiplier=0.88
    text=Join Now
    colorAndOpacity {colorUseRule=UseColor_Specified, specifiedColor=<rgba 0.0048 0.004 0.0116 1>}
    font {
     bForceMonospaced=false
     bMaterialIsStencil=false
     fontMaterial=null
     fontObject=&HeadingNow
     letterSpacing=0
     monospacedWidth=1
     size=15.75
     skewAmount=0
     typefaceFontName=75 Medium
     outlineSettings {
      bApplyOutlineToDropShadows=false
      bMiteredCorners=false
      bSeparateFillAlpha=false
      outlineMaterial=null
      outlineSize=0
      outlineColor=<rgba 0 0 0 1>
     }
    }
    shadowOffset=<0 0>
   }
   slot {class=GridSlot, properties {row=1, verticalAlignment=VAlign_Top}}
  }
  SizeBox_1 {
   widgetClassPath=&SizeBox
   parent=TooltipRoot_Live
   widget=SizeBox_1
   properties {bOverride_MaxDesiredHeight=true, heightOverride=48, maxDesiredHeight=48, maxDesiredWidth=200, widthOverride=200}
   slot {class=GridSlot, properties {verticalAlignment=VAlign_Top}}
  }
  TournamentName_Live {
   widgetClassPath=&CommonTextBlock
   bIsVariable=true
   parent=SizeBox_1
   widget=TournamentName_Live
   properties {
    bIsScrollingEnabled=false
    lineHeightPercentage=0.9
    minDesiredWidth=290
    scrollStyle=&TextScrollStyle-Base_C
    text=fortnite performance elevation
    textOverflowPolicy=MultilineEllipsis
    textTransformPolicy=ToUpper
    wrapTextAt=300
    colorAndOpacity {colorUseRule=UseColor_Specified, specifiedColor=<rgba 0.0048 0.004 0.0116 1>}
    font {
     bForceMonospaced=false
     bMaterialIsStencil=false
     fontMaterial=null
     fontObject=&HeadingNow
     letterSpacing=0
     monospacedWidth=1
     size=15.75
     skewAmount=0
     typefaceFontName=76 Bold
     outlineSettings {
      bApplyOutlineToDropShadows=false
      bMiteredCorners=false
      bSeparateFillAlpha=false
      outlineMaterial=null
      outlineSize=0
      outlineColor=<rgba 0 0 0 1>
     }
    }
    shadowOffset=<0 0>
   }
   slot {class=SizeBoxSlot}
  }
  Image_Live {
   widgetClassPath=&Image
   bIsVariable=true
   widget=Image_Live
   properties {
    visibility=Collapsed
    brush {
     bIsDynamicallyLoaded=false
     drawAs=Image
     imageType=FullColor
     mirroring=NoMirror
     resourceName=None
     resourceObject=/Compete/MaterialInstances/MI_UI_LiveEvent
     tiling=NoTile
     imageSize=<32 32>
     margin=<ltrb 0 0 0 0>
     outlineSettings {
      bUseBrushTransparency=false
      roundingType=HalfHeightRadius
      width=0
      color {colorUseRule=UseColor_Specified, specifiedColor=<rgba 0 0 0 0>}
      cornerRadii=<0 0 0 0>
     }
     tintColor {colorUseRule=UseColor_Specified, specifiedColor=<rgba 1 1 1 1>}
     uVRegion {bIsValid=false, max=<0 0>, min=<0 0>}
    }
   }
  }
 ]
}
